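# Sound cue for the power-up skill "not enough energy" sound, by its name an error cue for a failed skill use.
# A single wave player plays a lowered UI error wave at reduced volume, with a very large max distance.
# It is assigned to the Fort_LowHealth_PSM sound class.
# A concurrency override caps simultaneous instances at two, stopping the oldest.
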SoundCue /Game/Sounds/Fort_GamePlay_Sounds/PowerUps/Skill_Not_Enough_Energy_Cue
root=WavePlayer_0
nodes [{id=WavePlayer_0, properties {soundWaveAssetPtr=&/Fort_UI_Error_01_lower}}]
properties {
 bOverrideConcurrency=true
 maxDistance=2097152
 soundClassObject=/Game/Sounds/FortSoundClassesAndModes/Fort_LowHealth_PSM
 volumeMultiplier=0.1
 concurrencyOverrides {maxCount=2, resolutionRule=StopOldest}
}
